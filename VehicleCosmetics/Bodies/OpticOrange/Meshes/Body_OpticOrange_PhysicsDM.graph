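# Physics asset for the OpticOrange vehicle body mesh in the vehicle cosmetics content, defining its collision shape.
# A single body on the root bone holds one car-sized box, offset forward and raised from the bone origin.
# It has no constraints, so the asset describes one simple collision volume with no jointed parts.
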PhysicsAsset /VehicleCosmetics/Bodies/OpticOrange/Meshes/Body_OpticOrange_PhysicsDM
bodies {root [{shapeType=Box, extentX=444, extentY=216, extentZ=132, shapeName=root_box, center=<35.1689 0 104.9395>}]}
constraints []
